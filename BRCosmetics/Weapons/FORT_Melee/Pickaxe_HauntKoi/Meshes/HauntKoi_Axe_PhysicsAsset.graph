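# Physics asset for the HauntKoi pickaxe, a two-handed melee weapon cosmetic mesh.
# It holds a single capsule body on the Melee_TwoHanded_Handle bone, offset and rotated to cover the weapon.
# There are no constraints, so the axe simulates and collides as one rigid shape.
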PhysicsAsset /BRCosmetics/Weapons/FORT_Melee/Pickaxe_HauntKoi/Meshes/HauntKoi_Axe_PhysicsAsset
bodies {Melee_TwoHanded_Handle [{shapeType=Capsule, length=70.9511, radius=33.6596, center=<0.8205 2.6814 27.5347>, rotation=<rot -0.7905 -90 -0.0478>}]}
constraints []
